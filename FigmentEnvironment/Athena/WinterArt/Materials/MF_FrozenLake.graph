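# Material function for a frozen lake surface, outputting ice-and-snow material attributes plus a noise mask.
# World-projected swirl and crack normals form the ice, modulated by Fresnel and vertex green and blue.
# Snow from the Athena snow and biome color functions covers it through height lerps driven by vertex red and SnowIceTransition, with a bluish edge.
# The mask noise switches to larger tiling with depth, overridable by CustomStaticDepth.
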MaterialFunction /FigmentEnvironment/Athena/WinterArt/Materials/MF_FrozenLake
graph {nodes [
  n1 = MaterialFunctionCall() {materialFunction=/FigmentEnvironment/LM_8_1_1/Athena/WinterArt/Materials/MF_SharedWinter}
  n2 = Constant() {r=0.65}
  n3 = Constant() {r=0.125}
  n4 = Fresnel(Normal=VertexNormalWS()) {baseReflectFraction=0, exponent=1}
  n5 = Constant3Vector() {constant=<rgba 0 0 1 0>}
  n6 = ComponentMask(None=Divide(A=WorldPosition().XYZ, B=ScalarParameter() {defaultValue=9430.6709, parameterName=SnowMap})) {g=true, r=true}
  n7 = TextureSample(UVs=@n6) {samplerType=SAMPLERTYPE_Normal, texture=/FigmentEnvironment/Migrated_From_LM_7_00/Environments/Foliage/Textures/T_SwirlNoise_N}
  n8 = ComponentMask(None=Divide(A=WorldPosition().XYZ, B=Constant() {r=1000})) {g=true, r=true}
  n9 = ScalarParameter() {defaultValue=12, group=IceLakeSurf, parameterName=CracksNormalBoost}
  n10 = Constant2Vector() {g=1, r=1}
  n11 = Multiply(
   A=TextureSample(UVs=@n8) {constMipValue=4, samplerType=SAMPLERTYPE_Normal, texture=/Game/Athena/WinterArt/Textures/Cracks_N}.RGB
   B=AppendVector(A=Multiply(A=@n9, B=@n10), B=Constant() {r=1})
  )
  n12 = TextureSample(UVs=@n6) {samplerType=SAMPLERTYPE_LinearGrayscale, texture=/FigmentEnvironment/LM_8_1_1/Effects/Fort_Effects/BluGlow/Violentium/T_SwirlNoise_N_height}
  n13 = VertexColor()
  n14 = Multiply(A=@n12.R, B=@n13.G)
  n15 = OneMinus(None=@n13.B)
  n16 = OneMinus(None=TextureSample(UVs=@n8) {samplerType=SAMPLERTYPE_Grayscale, texture=/Game/Athena/WinterArt/Textures/Cracks}.R)
  n17 = Constant() {r=0.25}
  n18 = MaterialFunctionCall(
   A=@n5
   B=MaterialFunctionCall(
    Normal=MaterialFunctionCall(BaseNormal=@n7.RGB, AdditionalNormal=@n11) {materialFunction=&BlendAngleCorrectedNormals}.Result
    Flatness=LinearInterpolate(Alpha=@n4) {constA=1, constB=0}
   ) {materialFunction=&FlattenNormal}.Result
   Transition Phase=Constant() {r=0.41}
   Height Texture=Add(A=Add(A=@n14, B=@n15), B=Multiply(A=@n16, B=@n17))
   Contrast=Constant() {r=-3}
  ) {materialFunction=&HeightLerp}
  n19 = LinearInterpolate(A=Multiply(A=@n1.IceColor, B=LinearInterpolate(A=@n2, B=@n3, Alpha=@n4)), B=@n1.IceColor, Alpha=Multiply(A=@n18.Alpha, B=@n4))
  n20 = GetMaterialAttributes(None=MaterialFunctionCall!(Snow=MaterialFunctionCall!(UseLandscapeCoordsOnSwitch) {materialFunction=/FigmentEnvironment/LM_8_1_1/Athena/Environments/Landscape/MaterialFunctions/Snow/MF_Athena_Snow_01}.Material) {materialFunction=/FigmentEnvironment/LM_8_1_1/Athena/Environments/Landscape/MaterialFunctions/Snow/MF_Figment_BlendSnowBiomeColor}.Result) {attributeGetTypes [
    #1
    #2
    #3
    #4
   ]}
  n21 = ComponentMask(None=Divide(A=WorldPosition().XYZ, B=Constant() {r=500})) {g=true, r=true}
  n22 = Constant() {r=3}
  n23 = ScalarParameter() {parameterName=CustomStaticDepth}
  n24 = LinearInterpolate(
   A=MaterialFunctionCall!(World Position) {materialFunction=/Engine/Functions/Engine_MaterialFunctions02/Utility/DepthFromWorldPosition}.Result
   B=@n23
   Alpha=Saturate(None=@n23)
  )
  n25 = Constant() {r=5000}
  n26 = LinearInterpolate(
   A=TextureSample(UVs=@n21) {samplerType=SAMPLERTYPE_Grayscale, texture=/FigmentEnvironment/Athena/WinterArt/Textures/T_SwirlNoise_H2}.R
   B=TextureSample(UVs=Divide(A=@n21, B=@n22)) {samplerType=SAMPLERTYPE_Grayscale, texture=&T_SwirlNoise_H2}.RGB
   Alpha=Saturate(None=Divide(A=@n24, B=@n25))
  )
  n27 = MaterialFunctionCall(Transition Phase=OneMinus(None=VertexColor().R), Height Texture=@n26, Contrast=Constant() {r=6.5231}) {materialFunction=&HeightLerp}
  n28 = ScalarParameter() {defaultValue=0.5386, parameterName=SnowIceTransition}
  n29 = Add(A=@n12.R, B=Multiply(A=@n26, B=ScalarParameter() {defaultValue=0.2, parameterName=SnowEdgeContrast}))
  n30 = MaterialFunctionCall(Transition Phase=Add(A=@n28) {constB=0.025}, Height Texture=@n29, Contrast=Constant() {r=-5}) {materialFunction=&HeightLerp}
  n31 = LinearInterpolate(A=Constant3Vector() {constant=<rgba 0.1553 0.3941 0.5885 0>}, B=@n20.BaseColor, Alpha=Saturate(None=Add(A=@n27.Alpha, B=@n30.Alpha)))
  n32 = MaterialFunctionCall(Transition Phase=@n28, Height Texture=@n29, Contrast=Constant() {r=-5}) {materialFunction=&HeightLerp}
  n33 = Saturate(None=Add(A=@n27.Alpha, B=@n32.Alpha))
  n34 = LinearInterpolate(
   A=LinearInterpolate(A=Constant() {r=0.125}, B=Constant() {r=0.25}, Alpha=@n18.Alpha)
   B=LinearInterpolate(B=@n20.Roughness, Alpha=@n30.Alpha) {constA=0.35}
   Alpha=@n32.Alpha
  )
  n35 = LinearInterpolate(A=@n18.Results, B=@n5, Alpha=@n4)
  FunctionOutput(None=MakeMaterialAttributes!(
    BaseColor=LinearInterpolate(A=@n19, B=@n31, Alpha=@n33)
    Metallic
    Specular=LinearInterpolate(B=@n20.Specular, Alpha=@n32.Alpha) {constA=0.2, constB=0.5}
    Roughness=LinearInterpolate(A=@n34, B=@n20.Roughness, Alpha=@n27.Alpha)
    Anisotropy
    EmissiveColor
    Opacity
    OpacityMask
    Normal=LinearInterpolate(A=@n35, B=@n20.Normal, Alpha=@n33)
    Tangent
    WorldPositionOffset
    SubsurfaceColor
    ClearCoat
    ClearCoatRoughness
    AmbientOcclusion
    Refraction
    CustomizedUVs_0
    CustomizedUVs_1
    CustomizedUVs_2
    CustomizedUVs_3
    CustomizedUVs_4
    CustomizedUVs_5
    CustomizedUVs_6
    CustomizedUVs_7
    PixelDepthOffset
    ShadingModel
    Displacement
   ))
  FunctionOutput(None=@n26) {outputName=Mask}
 ]}
