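# M_IceWater is a frozen lake surface material in the Figment environment's Athena WinterArt set.
# It uses material attributes with Default Lit shading, and its only node is a call to the MF_FrozenLake material function, so the ice look is defined in that function.
# Automatic usage flag setting in the editor is turned off.
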Material /FigmentEnvironment/Athena/WinterArt/Materials/M_IceWater
graph {
 properties {bAutomaticallySetUsageInEditor=false, bUseMaterialAttributes=true, pixelDepthOffsetMode=PDOM_Legacy, usedShadingModels=Default Lit}
 nodes [MaterialFunctionCall() {materialFunction=&/MF_FrozenLake}]
}
